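# Physics asset for the InflateMate mic stand pickaxe cosmetic, giving its microphone ribbon physical bodies and joints.
# A capsule on the handle bone anchors a chain of four small capsules along the dyn_ribbon bones.
# Constraints link each segment to the previous one with twist locked and limit angles set for swing, loosest at the handle joint.
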PhysicsAsset /SparksCosmetics/Instrument/MicStand/InflateMate/Meshes/Pickaxe/InflateMate_Mic_Pickaxe_Physics
bodies {
 Mic_dyn_ribbon_01 [{shapeType=Capsule, length=2.3236, radius=0.6, shapeName=dyn_ribbon_01_capsule, center=<2 0 0>, rotation=<rot -90 0 0>}]
 Mic_dyn_ribbon_02 [{shapeType=Capsule, length=1.5556, radius=0.6, shapeName=dyn_ribbon_02_capsule, center=<2 0 0>, rotation=<rot -90 0 0>}]
 Mic_dyn_ribbon_03 [{shapeType=Capsule, length=2.6218, radius=0.6, shapeName=dyn_ribbon_03_capsule, center=<2 0 0>, rotation=<rot -90 0 0>}]
 Mic_dyn_ribbon_04 [{shapeType=Capsule, length=3.7373, radius=0.7056, shapeName=dyn_ribbon_04_capsule, center=<2 0 0>, rotation=<rot -90 0 0>}]
 handle [{shapeType=Capsule, length=10, radius=3, center=<0 0 -5.5205>}]
}
constraints [
 {
  bone1Name=Mic_dyn_ribbon_04
  bone2Name=Mic_dyn_ribbon_03
  swing1LimitDegrees=35
  swing2LimitDegrees=20
  twistLimitDegrees=45
  twistMotion=Locked
 }
 {
  bone1Name=Mic_dyn_ribbon_03
  bone2Name=Mic_dyn_ribbon_02
  swing1LimitDegrees=30
  swing2LimitDegrees=20
  twistLimitDegrees=45
  twistMotion=Locked
 }
 {
  bone1Name=Mic_dyn_ribbon_02
  bone2Name=Mic_dyn_ribbon_01
  swing1LimitDegrees=35
  swing2LimitDegrees=35
  twistLimitDegrees=45
  twistMotion=Locked
 }
 {bone1Name=Mic_dyn_ribbon_01, bone2Name=handle, swing1LimitDegrees=45, swing2LimitDegrees=45, twistLimitDegrees=45, twistMotion=Locked}
]
